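# Unlit, two-sided translucent ribbon trail material for the Glider_Drift VinderTech glider chute effect, enabled for mesh particles.
# Panning smoke and laser textures use edge-warped UVs, with stepped-time noise jittering the rounded smoke, then sphere masks and particle alpha shape them.
# Muzzle noise and smoke drive a curve-atlas color, blended toward pink along V.
# A distorted scrolling cloud, controlled by dynamic parameters, builds a V-masked vector transformed from local space.
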Material /BRCosmetics/Gadgets/Assets/VinderTech_GliderChute/Glider_Drift/FX/M_RibbonTrail_1
graph {
 properties {
  bUsedWithMeshParticles=true
  blendMode=BLEND_Translucent
  pixelDepthOffsetMode=PDOM_Legacy
  shadingModel=MSM_Unlit
  twoSided=true
  usedShadingModels=Unlit
 }
 nodes [
  n1 = TextureSample(UVs=Panner(Coordinate=TextureCoordinate() {uTiling=0.4, vTiling=0.55}) {speedX=0.18, speedY=-0.45}) {samplerType=SAMPLERTYPE_LinearColor, texture=/Game/Effects/Fort_Effects/Textures/Tile/T_Raygun_Muzzle_02}
  n2 = MaterialFunctionCall(Float2=TextureCoordinate() {vTiling=0.49}) {materialFunction=&BreakOutFloat2Components}
  n3 = Subtract(A=@n2.G) {constB=0.5}
  n4 = Power(Base=Multiply(A=Abs(None=@n3)) {constB=2}) {constExponent=3.5}
  n5 = LinearInterpolate(Alpha=@n2.R) {constA=-1, constB=2}
  n6 = Clamp(None=AppendVector(A=LinearInterpolate(A=@n2.R, B=@n5, Alpha=@n4), B=@n2.G))
  n7 = TextureCoordinate() {uTiling=0.46, vTiling=0.2476}
  n8 = Time()
  n9 = Divide(A=Ceil(None=Multiply(A=@n8) {constB=16})) {constB=4}
  n10 = ComponentMask(None=TextureSample(UVs=Panner(Coordinate=@n7, Time=@n9) {speedX=0.4, speedY=0.4}) {samplerType=SAMPLERTYPE_LinearColor, texture=/Game/Global/Textures/Default/Noise/TilingNoise06_N_uncompressed}.RGB) {g=true, r=true}
  n11 = Panner(Coordinate=Add(
    A=MaterialFunctionCall(XY=@n6) {materialFunction=&Swizzle}.YX
    B=Multiply(A=@n10) {constB=0.1}
   )) {speedX=-1}
  n12 = Multiply(
   A=Round(None=TextureSample(UVs=@n11) {texture=/Game/Effects/Fort_Effects/Textures/Smoke/T_Smoke_ThinTrail_Packed}.B)
   B=SphereMask(A=TextureCoordinate(), B=Constant2Vector() {r=0.5}) {attenuationRadius=0.3, hardnessPercent=0}
  )
  n13 = CurveAtlasRowParameter(CurveTime=Add(A=@n1.G, B=Multiply(A=@n12) {constB=1.5})) {
   atlas=&AthenaCurveAtlas01
   curve=/Game/Athena/LinearColorCurves/Curve_Glider_Drift
   defaultValue=105
   parameterName=ParamC
  }
  n14 = ComponentMask(None=TextureCoordinate()) {g=true}
  MaterialFunctionCall(Emissive=LinearInterpolate(A=@n13, B=Constant3Vector() {constant=<rgba 1 0.125 0.6163 1>}, Alpha=Multiply(A=@n14, B=@n14))) {materialFunction=&MF_FX_EmissiveAdjustments}
  n15 = MaterialFunctionCall(Float2=TextureCoordinate()) {materialFunction=&BreakOutFloat2Components}
  n16 = Abs(None=Subtract(A=@n15.G) {constB=0.5})
  n17 = LinearInterpolate(
   A=LinearInterpolate(Alpha=@n15.R) {constA=0.2, constB=0.8}
   B=LinearInterpolate(Alpha=@n15.R) {constA=-1.6, constB=2.6}
   Alpha=Power(Base=Multiply(A=@n16) {constB=2}) {constExponent=2.8}
  )
  n18 = Multiply(A=@n15.G) {constB=0.5}
  n19 = MaterialFunctionCall(XY=Clamp(None=AppendVector(A=@n17, B=@n18))) {materialFunction=&Swizzle}
  n20 = Power(Base=TextureSample(UVs=Panner(Coordinate=@n19.YX) {speedX=-1}) {texture=/Game/Effects/Fort_Effects/Textures/Tile/T_LaserTrail}.R) {constExponent=3}
  n21 = Multiply(
   A=SphereMask(A=TextureCoordinate(), B=Constant2Vector() {r=0.5}) {attenuationRadius=1, hardnessPercent=60}
   B=Add(A=Multiply(A=@n1.G, B=@n20), B=@n20)
  )
  Multiply(A=Add(A=@n12, B=Multiply(A=@n21) {constB=3}), B=ParticleColor().A)
  n22 = Panner(Coordinate=TextureCoordinate() {uTiling=0.6, vTiling=0.8}) {speedX=0.1, speedY=0.2}
  n23 = Multiply(
   A=ComponentMask(None=TextureSample(UVs=@n22) {samplerType=SAMPLERTYPE_Normal, texture=&T_Weap_Distort_N_02}.RGB) {g=true, r=true}
   B=Constant() {r=0.04}
  ) {constB=0.18}
  n24 = TextureCoordinate()
  n25 = DynamicParameter() {paramNames [Random,Param2,Param3,Param4]}
  n26 = Constant() {r=3}
  n27 = Divide(A=Add(A=ComponentMask(None=@n24) {g=true}, B=@n25.Random), B=AppendVector(A=Add(A=@n25.Param3, B=@n26), B=Constant() {r=64}))
  n28 = Multiply(A=Time(), B=Add(A=@n25.Param2, B=Constant() {r=18}))
  n29 = AppendVector(A=Constant() {r=-14}, B=Constant() {r=64})
  n30 = TextureSample(UVs=Panner(Coordinate=Add(A=@n23, B=@n27), Time=Divide(A=@n28) {constB=4}, Speed=Divide(B=@n29))) {texture=/Game/Effects/Fort_Effects/PlayerHealthDamage/T_DBNO_Cloud}
  n31 = Multiply(A=@n30.R, B=Constant() {r=0.5}) {constB=0.5}
  n32 = Multiply(A=@n30.G, B=Constant() {r=1}) {constB=5}
  n33 = Power(Base=AppendVector(A=AppendVector(A=@n31, B=@n32), B=Multiply(A=@n30.R) {constB=0.75}), Exp=Constant() {r=1.8}) {constExponent=1.2}
  n34 = Multiply(A=Constant() {r=5}, B=DynamicParameter() {paramNames [Random,Param2,Param3,Param4]}.Param4)
  n35 = ComponentMask(None=TextureCoordinate()) {g=true}
  n36 = Multiply(A=Power(Base=@n35) {constExponent=0.85}, B=SphereMask(A=@n35, B=Constant()) {attenuationRadius=1.35, hardnessPercent=0})
  Transform(None=Multiply(A=Multiply(A=@n33, B=@n34), B=LinearInterpolate(Alpha=@n36) {constB=1.6})) {transformSourceType=TRANSFORMSOURCE_Local}
 ]
}
